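# Default Lit skeletal-mesh material for the JamEvilOrnament03 skin of the Jam vehicle body.
# An RGBA mask texture layers grey, dark and red zones across base color, metallic and roughness, with a procedural tiled-dot "Grate" pattern detailing the alpha zone.
# The skin normal's XY adds onto a BodyNormalLOD parameter normal, then blends to the grate normal by mask alpha.
# The attributes pass through shared vehicle damage and de-rez transition functions.
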Material /VehicleCosmetics/Bodies/Jam/Skins/JamEvilOrnament03/M_Skin_JamEvilOrnament03
graph {
 properties {
  bAutomaticallySetUsageInEditor=false
  bUseMaterialAttributes=true
  bUsedWithMeshDeformer=true
  bUsedWithSkeletalMesh=true
  usedShadingModels=Default Lit
 }
 nodes [
  n1 = Constant3Vector() {constant=<rgba 0.2865 0.2865 0.2865 1>}
  n2 = TextureSample() {samplerType=SAMPLERTYPE_LinearColor, texture=&/JamEvilOrnament03_RGB}
  n3 = LinearInterpolate(A=@n1, B=Constant() {r=0.01}, Alpha=@n2.B)
  n4 = Constant3Vector() {constant=<rgba 0.4508 0.0013 0 1>}
  n5 = NamedRerouteUsage()
  n6 = LinearInterpolate(
   A=LinearInterpolate(A=LinearInterpolate(A=@n3, B=@n4, Alpha=@n2.R), B=LinearInterpolate(Alpha=@n5) {constB=0.1}, Alpha=@n2.A) {constB=0.1}
   B=Multiply(A=@n1) {constB=0.6}
   Alpha=@n2.G
  )
  n7 = LinearInterpolate(A=LinearInterpolate(A=LinearInterpolate(Alpha=@n2.G) {constA=0.8, constB=0.8}, Alpha=@n2.R), Alpha=@n2.A)
  n8 = LinearInterpolate(Alpha=@n2.G) {constA=0.1, constB=0.5}
  n9 = LinearInterpolate(
   A=LinearInterpolate(A=LinearInterpolate(A=@n8, Alpha=@n2.R) {constB=0.25}, Alpha=@n2.A) {constB=0.25}
   B=LinearInterpolate(Alpha=NamedRerouteUsage()) {constA=1, constB=0.2}
   Alpha=@n2.A
  )
  n10 = TextureSample() {samplerType=SAMPLERTYPE_Normal, texture=&/JamEvilOrnament03_normal}
  n11 = Constant3Vector() {constant=<rgba 1 1 0 0>}
  n12 = LinearInterpolate(
   A=Add(
    A=Multiply(A=@n10.RGB, B=@n11)
    B=TextureSampleParameter2D() {parameterName=BodyNormalLOD, samplerType=SAMPLERTYPE_Normal, texture=/VehicleCosmetics/Textures/T_BlankNormal_B}.RGB
   )
   B=LinearInterpolate(A=Constant3Vector() {constant=<rgba 1 0 0.9067 1>}, B=Constant3Vector() {constant=<rgba 0 0 1 0>}, Alpha=NamedRerouteUsage())
   Alpha=@n2.A
  )
  MaterialFunctionCall(In=MaterialFunctionCall(Material=MakeMaterialAttributes!(
     BaseColor=@n6
     Metallic=@n7
     Specular
     Roughness=@n9
     Anisotropy
     EmissiveColor
     Opacity
     OpacityMask
     Normal=@n12
     Tangent
     WorldPositionOffset
     SubsurfaceColor
     ClearCoat
     ClearCoatRoughness
     AmbientOcclusion
     Refraction
     CustomizedUVs_0
     CustomizedUVs_1
     CustomizedUVs_2
     CustomizedUVs_3
     CustomizedUVs_4
     CustomizedUVs_5
     CustomizedUVs_6
     CustomizedUVs_7
     PixelDepthOffset
     ShadingModel
     Displacement
    )) {materialFunction=&MF_BaseVehicle_Damage}.Result) {materialFunction=&MF_VehicleDeRez_Transition}
  n13 = TextureCoordinate()
  n14 = Subtract(A=Frac(None=Multiply(A=@n13) {constB=500})) {constB=0.5}
  n15 = DotProduct(A=@n14, B=@n14)
  NamedRerouteDeclaration(None=Saturate(None=ConstantBiasScale(None=@n15) {bias=-0.1, scale=6})) {name=Grate, nodeColor=<rgba 0.7344 0 1 1>}
 ]
}
